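# Surface material for the BP12 front-end elevator doors, animating door opening, light strips and an optional chrome finish.
# Open and OpenMouseOver parameters drive vertex offsets, switching between vertex-color-directed sliding and pivot rotation with corrected normals.
# A world-space distance mask around CenterBlingLocation blends chrome color, metallic and roughness, adding a timed Fresnel-boosted shimmer.
# Emissive switches between a masked texture and a light strip that pans up or down and fades on opening.
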Material /Game/Environments/FrontEnd/Meshes/BP12/M_ElevatorDoors
graph {
 properties {pixelDepthOffsetMode=PDOM_Legacy, usedShadingModels=Default Lit}
 nodes [
  Add()
  n1 = WorldPosition()
  n2 = Multiply(A=ComponentMask(None=@n1.XYZ) {r=true}, B=ScalarParameter() {defaultValue=0.1333, parameterName=TDistort})
  n3 = AppendVector(A=@n2, B=@n2)
  n4 = VectorParameter() {parameterName=CenterBlingLocation, defaultValue=<rgba 187 -2560 441.4479 1>}
  n5 = Constant3Vector() {constant=<rgba 0 1 1 0>}
  n6 = Multiply(A=@n4.RGB, B=@n5)
  Add(A=AppendVector(A=Constant(), B=@n3), B=@n6)
  n7 = ScalarParameter() {defaultValue=2000, parameterName=Size}
  MaterialFunctionCall(
   TextureObject=TextureObject() {samplerType=SAMPLERTYPE_Grayscale, texture=/Game/Effects/Fort_Effects/Textures/T_EFX_CameraFlare_01}
   TextureSize=@n7
  ) {materialFunction=/Game/Weapons/WeaponSkins/Wraps/MaterialFunctions/WorldAlignedTexture}
  n8 = ScalarParameter() {parameterName=Open}
  n9 = ScalarParameter() {parameterName=OpenMouseOver}
  Max(A=@n8, B=@n9)
  n10 = ScalarParameter() {defaultValue=4, parameterName=Pause}
  n11 = Multiply(A=Time(), B=Divide(A=ScalarParameter() {defaultValue=1, parameterName=SpeedShimmer}, B=@n10))
  n12 = Saturate(None=Multiply(A=Frac(None=@n11), B=@n10))
  n13 = Divide(A=Add(A=ComponentMask(None=@n1.XYZ) {b=true, g=true}, B=@n3), B=ScalarParameter() {defaultValue=1800, parameterName=Tiling})
  Rotator(
   Coordinate=Add(A=Multiply(A=@n12) {constB=4}, B=Subtract(A=@n13) {constB=0.5})
   Time=ScalarParameter() {defaultValue=3.7957, parameterName=rot}
  )
  n14 = VertexColor()
  n15 = StaticSwitchParameter(True=@n14.R, False=OneMinus(None=@n14.R)) {parameterName=Flip}
  n16 = Transform(None=Multiply(A=VectorParameter() {parameterName=Direction, defaultValue=<rgba 1 0 0 1>}.RGB, B=LinearInterpolate(Alpha=@n15) {constA=-1})) {transformSourceType=TRANSFORMSOURCE_Local}
  n17 = Multiply(A=Add(A=@n14.R, B=@n14.R), B=ScalarParameter() {defaultValue=100, parameterName=MaxDistance})
  n18 = Saturate(None=Add(A=@n8, B=@n9))
  n19 = Constant3Vector() {constant=<rgba 0 0 1 0>}
  n20 = LinearInterpolate(B=ScalarParameter() {defaultValue=0.06, parameterName=RotationMax}, Alpha=@n18)
  n21 = LinearInterpolate(A=Multiply(A=@n20) {constB=-1}, B=Multiply(A=@n20, B=ScalarParameter() {defaultValue=1, parameterName=RotationMaxBias}), Alpha=@n15)
  n22 = Constant3Vector()
  StaticSwitchParameter(
   True=Multiply(A=Multiply(A=@n16, B=@n17), B=@n18)
   False=RotateAboutAxis(
    NormalizedRotationAxis=@n19
    RotationAngle=@n21
    PivotPoint=TransformPosition(None=@n22) {transformType=TRANSFORMPOSSOURCE_World}
    Position=WorldPosition().XYZ
   )
  ) {parameterName=IsSlidingDoor}
  n23 = TextureSampleParameter2D() {parameterName=Normal, samplerType=SAMPLERTYPE_Normal, texture=/Game/Environments/FrontEnd/Meshes/BP12/Textures/T_Elevator_Door_BP12_N}
  StaticSwitchParameter(
   True=@n23.RGB
   False=MaterialFunctionCall(
    BaseNormal=MaterialFunctionCall(Rotation Axis=@n19, Rotation Angle=@n21) {materialFunction=&FixRotateAboutAxisNormals}.New Tangent Space Vertex Normal
    AdditionalNormal=@n23.RGB
   ) {materialFunction=&BlendAngleCorrectedNormals}.Result
  ) {parameterName=IsSlidingDoor}
  n24 = TextureSampleParameter2D() {parameterName=Diffuse, texture=/Game/Environments/FrontEnd/Meshes/BP12/Textures/T_Elevator_Door_BP12_D}
  n25 = LinearInterpolate(
   A=VectorParameter() {parameterName=ColorChrome1, defaultValue=<rgba 0 0 0 1>}.RGB
   B=VectorParameter() {parameterName=ColorChrome2, defaultValue=<rgba 1 1 1 1>}.RGB
   Alpha=@n24.R
  )
  n26 = VectorParameter() {parameterName=ColorChrome3, defaultValue=<rgba 1 0.4428 0 1>}
  n27 = WorldPosition()
  n28 = TextureSample(UVs=Divide(A=ComponentMask(None=@n27.XYZ) {b=true, g=true}, B=@n7)) {samplerType=SAMPLERTYPE_Grayscale, texture=&T_EFX_CameraFlare_01}
  n29 = Multiply(A=Subtract(A=@n1.XYZ, B=@n4.RGB), B=VectorParameter() {parameterName=ScaleFalloff, defaultValue=<rgba 1 1 2.1923 1>}.RGB)
  n30 = Distance(A=Multiply(A=Add(A=@n29, B=@n4.RGB), B=@n5), B=@n6)
  n31 = Divide(A=@n30, B=ScalarParameter() {defaultValue=853.8812, parameterName=DistanceMask})
  n32 = ScalarParameter() {defaultValue=-0.8626, parameterName=MaskOffset}
  n33 = OneMinus(None=Saturate(None=Add(A=@n31, B=@n32)))
  n34 = ScalarParameter() {defaultValue=1, parameterName=MaskContrast}
  n35 = ScalarParameter() {parameterName=Chrome}
  n36 = Multiply(A=Saturate(None=Power(Base=@n33, Exp=@n34)), B=@n35)
  StaticSwitchParameter(True=LinearInterpolate(A=@n24.RGB, B=LinearInterpolate(A=@n25, B=@n26.RGB, Alpha=@n28.R), Alpha=@n36), False=@n24.RGB) {parameterName=UseChrome}
  n37 = TextureSampleParameter2D() {parameterName=Spec, samplerType=SAMPLERTYPE_Masks, texture=/Game/Environments/FrontEnd/Meshes/BP12/Textures/T_Elevator_Door_BP12_S}
  StaticSwitchParameter(True=LinearInterpolate(A=@n37.G, B=ScalarParameter() {defaultValue=1, parameterName=MetalChrome}, Alpha=@n36), False=@n37.G) {parameterName=UseChrome}
  n38 = ScalarParameter() {defaultValue=1, parameterName=RoughnessChrome}
  n39 = ScalarParameter() {parameterName=RoughnessChrome2}
  StaticSwitchParameter(True=LinearInterpolate(A=@n37.B, B=LinearInterpolate(A=@n38, B=@n39, Alpha=@n24.R), Alpha=@n36), False=@n37.B) {parameterName=UseChrome}
  n40 = Multiply(A=TextureCoordinate(), B=ScalarParameter() {defaultValue=0.6987, parameterName=ScalePanner})
  n41 = ScalarParameter() {defaultValue=0.5429, parameterName=PannerSpeed}
  n42 = ScalarParameter() {parameterName=EnableLightpanUp}
  n43 = Ceil(None=@n42)
  n44 = Multiply(A=@n41) {constB=-1}
  n45 = ScalarParameter() {parameterName=EnableLightpanDown}
  n46 = Ceil(None=@n45)
  n47 = AppendVector(A=Constant(), B=Add(A=LinearInterpolate(B=@n41, Alpha=@n43), B=LinearInterpolate(B=@n44, Alpha=@n46)))
  n48 = LinearInterpolate(
   A=ScalarParameter() {parameterName=LightBrightnessDefault}
   B=TextureSample(UVs=Panner(Coordinate=@n40, Speed=@n47)) {texture=/Game/Global/Textures/Default/Masks_and_Gradients/gradient_stasis}.R
   Alpha=Max(A=@n42, B=@n45)
  )
  n49 = VectorParameter() {parameterName=LightStripColor, defaultValue=<rgba 3.1586 4.118 5 1>}
  n50 = Multiply(A=@n18, B=ScalarParameter() {defaultValue=52.8468, parameterName=Boostlightfade})
  n51 = TextureSampleParameter2D() {parameterName=Emissive, samplerType=SAMPLERTYPE_Masks, texture=&T_Elevator_Door_BP12_S}
  n52 = ChannelMaskParameter(None=@n51.RGBA) {parameterName=EmissiveChannel}
  n53 = ScalarParameter() {defaultValue=2, parameterName=EmissivePower}
  n54 = StaticSwitchParameter(
   True=LinearInterpolate(A=Multiply(A=@n48, B=@n49.RGB), B=Constant(), Alpha=Saturate(None=@n50))
   False=Multiply(A=Power(Base=@n52.R, Exp=@n53), B=VectorParameter() {parameterName=EmissiveColor, defaultValue=<rgba 0 0 0 1>}.RGB)
  ) {parameterName=IsAnimated lightstrip}
  n55 = VectorParameter() {parameterName=ColorShimmer, defaultValue=<rgba 0.7659 0.8741 1 1>}
  n56 = VectorParameter() {parameterName=ColorShimmer2, defaultValue=<rgba 0.7413 0.8271 1 1>}
  n57 = Divide(A=@n30, B=ScalarParameter() {defaultValue=2224.8945, parameterName=DistanceBling})
  n58 = ComponentMask(None=Add(A=@n12, B=Subtract(A=@n57) {constB=0.5})) {r=true}
  n59 = ScalarParameter() {defaultValue=2.4393, parameterName=Fresnel}
  n60 = Multiply(
   A=Power(Base=Sine(None=@n58), Exp=ScalarParameter() {defaultValue=3, parameterName=ShimmerContrast})
   B=LinearInterpolate(
    A=ScalarParameter() {defaultValue=0.6, parameterName=Shimmer}
    B=ScalarParameter() {defaultValue=80, parameterName=Shimmer2}
    Alpha=Fresnel(ExponentIn=@n59) {baseReflectFraction=0}
   )
  )
  n61 = Multiply(A=Multiply(A=LinearInterpolate(A=@n55.RGB, B=@n56.RGB, Alpha=@n60), B=@n60), B=OneMinus(None=Saturate(None=@n31)))
  StaticSwitchParameter(True=Add(A=Multiply(A=@n61, B=@n35), B=@n54), False=@n54) {parameterName=UseChrome}
  VectorParameter() {parameterName=Color, defaultValue=<rgba 0.4583 0.4583 0.4583 1>}
 ]
}
